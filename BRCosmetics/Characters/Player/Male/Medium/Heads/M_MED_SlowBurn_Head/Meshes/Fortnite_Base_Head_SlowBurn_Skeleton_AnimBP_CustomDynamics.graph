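# Post-process style AnimBP for the SlowBurn head, adding ear and jawline fur motion.
# A Control Rig feeds additive ear rotations driven by frown curves, then spring-based AnimDynamics on ears and fur.
# EventGraph reads the left frown curve from the owner's frontend or player AnimInstance, the right from this instance.
# It also computes frontend-idle bypass weights, but their DefaultSlot ear blend is disconnected from the output.
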AnimBlueprint /BRCosmetics/Characters/Player/Male/Medium/Heads/M_MED_SlowBurn_Head/Meshes/Fortnite_Base_Head_SlowBurn_Skeleton_AnimBP_CustomDynamics
target_skeleton=/Game/Characters/Player/Male/Medium/Heads/M_MED_SlowBurn_Head/Meshes/Fortnite_Base_Head_SlowBurn_Skeleton
graphs [
 AnimGraph {nodes [
   Layered blend per bone(
    BasePose=Use cached pose 'postDynamics'()
    BlendPoses_0=Slot 'DefaultSlot'()
    BlendWeights_0=Get isPlayingFrontendIdle() -> (isPlayingFrontendIdle:float)
   ) {layerSetup [{branchFilters [{boneName=dyn_L_ear}, {boneName=dyn_R_ear}]}]}
   n1 = Control Rig(Source=Input Pose()) {
    alphaBoolBlend {blendOption=Linear}
    controlRigAssetReference {blueprintRigClass=&/M_MED_SlowBurn_Head_CtrlRig.M_MED_SlowBurn_Head_CtrlRig_C}
   }
   n2 = "Transform (Modify) Bone - Bone: dyn_L_ear"(
    ComponentPose=Local To Component(LocalPose=@n1)
    Alpha=Get L frown curve alpha() -> (L frown curve alpha:float)
    Translation="0,0,0"
    Rotation="0,5,70"
    Scale="1,1,1"
   ) {
    rotationMode=BMM_Additive
    rotationSpace=BCS_BoneSpace
    alphaBoolBlend {blendOption=Linear}
    alphaScaleBiasClamp {bInterpResult=true, interpSpeedDecreasing=20, interpSpeedIncreasing=20}
    boneToModify {boneName=dyn_L_ear}
    rotation {roll=70, yaw=5}
   }
   n3 = "Transform (Modify) Bone - Bone: dyn_R_ear"(ComponentPose=@n2, Alpha=Get R frown curve alpha() -> (R frown curve alpha:float), Translation="0,0,0", Rotation="0,0,70", Scale="1,1,1") {
    rotationMode=BMM_Additive
    rotationSpace=BCS_BoneSpace
    alphaBoolBlend {blendOption=Linear}
    alphaScaleBiasClamp {bInterpResult=true, interpSpeedDecreasing=20, interpSpeedIncreasing=20}
    boneToModify {boneName=dyn_R_ear}
    rotation {roll=70}
   }
   n4 = Anim Dynamics - Bone: dyn_L_ear(ComponentPose=@n3, Alpha=1, ExternalForce="0,0,0") {
    angularDampingOverride=1
    angularSpringConstant=8000
    bAngularSpring=true
    bOverrideAngularDamping=true
    bUseGravityOverride=true
    simulationSpace=RootRelative
    alphaBoolBlend {blendOption=Linear}
    boundBone {boneName=dyn_L_ear}
    componentLinearAccScale=<1 1 1>
    componentLinearVelScale=<1 1 1>
    gravityOverride {z=-980}
    physicsBodyDefinitions [{
      boundBone {boneName=dyn_L_ear}
      boxExtents=<15 15 15>
      constraintSetup {angularLimitsMax=<5 5 5>, angularLimitsMin=<-5 -5 -5>, angularTarget=<1 0 0>}
      localJointOffset=<1 0 0>
     }]
   }
   n5 = Anim Dynamics - Bone: dyn_R_ear(ComponentPose=@n4, Alpha=1, ExternalForce="0,0,0") {
    angularDampingOverride=1
    angularSpringConstant=8000
    bAngularSpring=true
    bOverrideAngularDamping=true
    bUseGravityOverride=true
    simulationSpace=RootRelative
    alphaBoolBlend {blendOption=Linear}
    boundBone {boneName=dyn_R_ear}
    componentLinearAccScale=<1 1 1>
    componentLinearVelScale=<1 1 1>
    gravityOverride {z=-980}
    physicsBodyDefinitions [{
      boundBone {boneName=dyn_R_ear}
      boxExtents=<15 15 15>
      constraintSetup {angularLimitsMax=<5 5 5>, angularLimitsMin=<-5 -5 -5>, angularTarget=<1 0 0>}
      localJointOffset=<-1 0 0>
     }]
   }
   n6 = Anim Dynamics - Bone: dyn_L_jawlineFur_front(ComponentPose=@n5, Alpha=1, ExternalForce="0,0,0") {
    angularDampingOverride=0.9999
    angularSpringConstant=2000
    bAngularSpring=true
    bOverrideAngularDamping=true
    bUseGravityOverride=true
    simulationSpace=RootRelative
    alphaBoolBlend {blendOption=Linear}
    boundBone {boneName=dyn_L_jawlineFur_front}
    componentLinearAccScale=<1 1 1>
    componentLinearVelScale=<1 1 1>
    gravityOverride {z=-980}
    physicsBodyDefinitions [{
      boundBone {boneName=dyn_L_jawlineFur_front}
      boxExtents=<11 11 11>
      constraintSetup {angularLimitsMax=<10 10 10>, angularLimitsMin=<-10 -10 -10>, angularTarget=<1 0 0>}
      localJointOffset=<1 0 0>
     }]
   }
   n7 = Anim Dynamics - Bone: dyn_L_jawlineFur_back(ComponentPose=@n6, Alpha=1, ExternalForce="0,0,0") {
    angularDampingOverride=0.9999
    angularSpringConstant=2000
    bAngularSpring=true
    bOverrideAngularDamping=true
    bUseGravityOverride=true
    simulationSpace=RootRelative
    alphaBoolBlend {blendOption=Linear}
    boundBone {boneName=dyn_L_jawlineFur_back}
    componentLinearAccScale=<1 1 1>
    componentLinearVelScale=<1 1 1>
    gravityOverride {z=-980}
    physicsBodyDefinitions [{
      boundBone {boneName=dyn_L_jawlineFur_back}
      boxExtents=<10 10 10>
      constraintSetup {angularLimitsMax=<10 10 10>, angularLimitsMin=<-10 -10 -10>, angularTarget=<1 0 0>}
      localJointOffset=<1 0 0>
     }]
   }
   n8 = Anim Dynamics - Bone: dyn_R_jawlineFur_front(ComponentPose=@n7, Alpha=1, ExternalForce="0,0,0") {
    angularDampingOverride=0.9999
    angularSpringConstant=2000
    bAngularSpring=true
    bOverrideAngularDamping=true
    bUseGravityOverride=true
    simulationSpace=RootRelative
    alphaBoolBlend {blendOption=Linear}
    boundBone {boneName=dyn_R_jawlineFur_front}
    componentLinearAccScale=<1 1 1>
    componentLinearVelScale=<1 1 1>
    gravityOverride {z=-980}
    physicsBodyDefinitions [{
      boundBone {boneName=dyn_R_jawlineFur_front}
      boxExtents=<11 11 11>
      constraintSetup {angularLimitsMax=<10 10 10>, angularLimitsMin=<-10 -10 -10>, angularTarget=<1 0 0>}
      localJointOffset=<-1 0 0>
     }]
   }
   n9 = Anim Dynamics - Bone: dyn_R_jawlineFur_back(ComponentPose=@n8, Alpha=1, ExternalForce="0,0,0") {
    angularDampingOverride=0.9999
    angularSpringConstant=2000
    bAngularSpring=true
    bOverrideAngularDamping=true
    bUseGravityOverride=true
    simulationSpace=RootRelative
    alphaBoolBlend {blendOption=Linear}
    boundBone {boneName=dyn_R_jawlineFur_back}
    componentLinearAccScale=<1 1 1>
    componentLinearVelScale=<1 1 1>
    gravityOverride {z=-980}
    physicsBodyDefinitions [{
      boundBone {boneName=dyn_R_jawlineFur_back}
      boxExtents=<10 10 10>
      constraintSetup {angularLimitsMax=<10 10 10>, angularLimitsMin=<-10 -10 -10>, angularTarget=<1 0 0>}
      localJointOffset=<-1 0 0>
     }]
   }
   Output Pose(Result=Component To Local(ComponentPose=@n9)) {layerGroup=None}
   Save cached pose 'postDynamics'() {cacheName=postDynamics}
   Get isNOT_PlayingFrontendIdle() -> (isNOT_PlayingFrontendIdle:double)
  ]}
 EventGraph {nodes [
   Event BlueprintUpdateAnimation()
   > n1 = Sequence()
   n2 = Branch(execute=@n1.then_1, Condition=Get bIsInFrontEnd())
   > n3 = Set isPlayingFrontendIdle(isPlayingFrontendIdle:float=GetCurveValue(self=GetOwnerFrontEndAnimInstance(), CurveName=Frontend_Idle_Bypass)) -> (Output_Get:float)
   Set isNOT_PlayingFrontendIdle(execute=[@n3.then, @n2.else], isNOT_PlayingFrontendIdle:double=float - float(A=1, B=@n3)) -> (Output_Get:double)
   n4 = Branch(execute=@n1.then_0, Condition=Get bIsInFrontEnd())
   > Set L frown curve alpha(L frown curve alpha:float=GetCurveValue(self=GetOwnerFrontEndAnimInstance(), CurveName=L_frown_pose)) -> (Output_Get:float)
   > Set R frown curve alpha(R frown curve alpha:float=GetCurveValue(CurveName=R_frown_pose)) -> (Output_Get:float)
   Set L frown curve alpha(
    execute=@n4.else
    L frown curve alpha=GetCurveValue(
     self=GetOwnerFortPlayerAnimInstance!(self:Custom Character Part Anim Instance Object Reference) -> (ReturnValue:Fort Player Anim Instance Object Reference)
     CurveName=L_frown_pose
    )
   )
   > Set R frown curve alpha(R frown curve alpha=GetCurveValue(CurveName=R_frown_pose))
  ]}
]
variables [
 isNOT_PlayingFrontendIdle {type=number}
 isPlayingFrontendIdle {type=number}
 L frown curve alpha {type=number}
 R frown curve alpha {type=number}
]
